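# Translucent, two-sided Default Lit material for the wings of the Kusanagi T3 vehicle body, usable on skeletal meshes.
# Cloud-noise-distorted UVs pan through cell noise, which is inverted, sharpened and masked by inverted, powered vertex green.
# That animated pattern is summed with the vertex mask and the boosted green channel of the Base Color texture parameter.
# The sum ends at a reroute; no material output connections appear, and TeamColor is unconnected.
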Material /VehicleCosmetics/Bodies/Kusanagi/Kusanagi_T3/Materials/MAT_KusanagiT3_Wings
graph {
 properties {
  bAutomaticallySetUsageInEditor=false
  bUsedWithSkeletalMesh=true
  blendMode=BLEND_Translucent
  previewMesh=/VehicleCosmetics/Bodies/Kusanagi/Kusanagi_T3/Meshes/Kusanagi_T3_Fender_SM
  twoSided=true
  usedShadingModels=Default Lit
 }
 nodes [
  n1 = TextureCoordinate() {uTiling=2, vTiling=2}
  n2 = Add(
   A=Multiply(A=TextureSample(UVs=@n1) {texture=/VehicleCosmetics/Wheels/_SharedTextures/FX/DCU_CloudDistort_Skin_T}.G) {constB=0.05}
   B=TextureCoordinate() {uTiling=3}
  )
  n3 = OneMinus(None=TextureSample(UVs=Panner(Coordinate=@n2) {speedY=0.25}) {texture=/VehicleCosmetics/Wheels/_SharedTextures/FX/Cell_Noise_Skins_T}.G)
  n4 = Power(Base=OneMinus(None=VertexColor().G)) {constExponent=3}
  n5 = Multiply(A=Multiply(A=Power(Base=@n3) {constExponent=8}) {constB=3}, B=@n4)
  n6 = TextureSampleParameter2D() {group=00_Base, parameterName=Base Color, texture=/VehicleCosmetics/Bodies/Kusanagi/Kusanagi_T3/Textures/T_KusanagiT3_E}
  Reroute(None=Add(A=Add(A=@n5, B=@n4), B=Multiply(A=@n6.G) {constB=3}))
  VectorParameter() {group=00_Base, parameterName=TeamColor, defaultValue=<rgba 1 0.6867 0.147 1>}
 ]
}
